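# Player survey response widget that shows a QR code and an optional "Open Link in Browser" button.
# BP_SetResponses substitutes the player's account ID as EpicId into each response text, encoding the first as the QR code and using the second as the button URL.
# The button collapses without a launchable URL; clicking it reports response index 1 selected and launches the link.
# Activation immediately marks the question completed.
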WidgetBlueprint /FortniteCoreUI/UI/PlayerSurvey/Widgets/PlayerSurveyQrCodeResponse
graphs [
 BP_GetDesiredFocusTarget {nodes [
   Get Desired Focus Target()
   > Return Node!(execute, ReturnValue:Widget Object Reference=Get Button_OpenLink() -> (Button_OpenLink:WBP UIKit Button Regular Object Reference))
  ]}
 BP_SetResponses {nodes [
   Set Responses!() -> (then, Responses:Array of Texts) {is=title K2Node_FunctionEntry}
   > n1 = Cast To FortPlayerController(Object=GetOwningPlayer())
   > n2 = GetDebugStringForUniqueId!(
    execute
    UniqueID:"Unique Net Id Structure (by ref)"=GetGameAccountId!(self:Fort Player Controller Object Reference=@n1) -> (ReturnValue:Unique Net Id Structure)
   ) -> (then, ReturnValue:String)
   > n3 = For Each Loop(Array:Array of Texts=Get Responses() -> (Responses:Array of Texts)) -> (Array Element:Text)
   n4 = Branch(execute=@n3.Completed, Condition=Is Valid Index!(TargetArray:Array of Texts=Get Responses(), IndexToTest:int) -> (ReturnValue:bool))
   n5 = Get Responses()
   > SetEncodedString!(
    execute
    self:WBP QRCode Generator Object Reference=Get WBP_QRCode() -> (WBP_QRCode:WBP QRCode Generator Object Reference)
    StringToEncode:String="To String (Text)"(InText="Get (a ref)"(Array:Array of Texts=@n5, Dimension 1=0) -> (Output:"Text (by ref)"))
   ) -> (then)
   > n6 = Branch(Condition=Is Valid Index(TargetArray=Get Responses(), IndexToTest=1))
   n7 = Get Responses()
   > SetButtonUrl!(execute, self, InButtonUrl:String="To String (Text)"(InText="Get (a ref)"(Array=@n7, Dimension 1=1))) -> (then)
   > Return Node(ReturnValue:bool=true)
   SetButtonUrl(execute=@n6.else)
   > Return Node(ReturnValue=true)
   Return Node(execute=@n4.else, ReturnValue=false)
   n8 = "Get (a ref)"(Array=Get Responses(), Dimension 1=@n3.Array Index)
   "Set Text (by ref)"!(execute:Exec=@n3.LoopBody, Target:"Text (by ref)"=@n8, Value:Text=Format Text!(Format=@n8, EpicId:String=@n2) -> (Result)) -> (then:Exec) {is=title K2Node_VariableSetRef}
   Return Node(execute=@n1.CastFailed, ReturnValue=false)
  ]}
 EventGraph {nodes [
   "On Clicked (Button_OpenLink)"!() -> (OutputDelegate, then, Button:Common Button Base Object Reference) {is=title K2Node_ComponentBoundEvent}
   > BroadcastOnResponseSelectionChanged!(execute, self:Fort Player Survey Blueprintable Response Widget Object Reference, ResponseIndex:int=1, bSelected:bool=true) -> (then)
   > LaunchUrl!(execute, self:Fort Player Survey Blueprintable Response Widget Object Reference, Url:String=Get ButtonUrl() -> (ButtonUrl:String)) -> (then)
   Event On Activated()
   > BroadcastOnQuestionCompletionUpdated!(execute, self:Fort Player Survey Blueprintable Response Widget Object Reference, bQuestionCompleted:bool=true) -> (then)
  ]}
 SetButtonUrl {nodes [
   n1 = SetButtonUrl!() -> (then, InButtonUrl:String) {is=title K2Node_FunctionEntry}
   > n2 = CanLaunchURL!(execute, URL:String=@n1) -> (then, ReturnValue:bool)
   > n3 = Set ButtonUrl(ButtonUrl:String=Select(Option 0:String, Option 1:String=Get InButtonUrl() -> (InButtonUrl:String), Index=@n2) -> (ReturnValue:String)) -> (Output_Get:String)
   > SetVisibility(
    self=Get Button_OpenLink()
    InVisibility=Select(Option 0:ESlateVisibility Enum=Visible, Option 1:ESlateVisibility Enum=Collapsed, Index=IsEmpty(InString=@n3)) -> (ReturnValue:ESlateVisibility Enum)
   )
  ]}
]
variables [ButtonUrl {type=string, default=""}]
widgets {
 info {
  inheritedWidgetCount=0
  parentClass=/Script/FortniteUI.FortPlayerSurveyBlueprintableResponseWidget
  rootWidgetClass=&VerticalBox
  widgetCount=4
 }
 widgets [
  VerticalBox_162 {widgetClassPath=&VerticalBox, widget=VerticalBox_162}
  Border_0 {
   widgetClassPath=&Border
   parent=VerticalBox_162
   widget=Border_0
   properties {
    background {
     bIsDynamicallyLoaded=false
     drawAs=Image
     imageType=FullColor
     mirroring=NoMirror
     resourceName=None
     resourceObject=/FortniteCoreUI/UI/PlayerSurvey/Materials/MI_PlayerSurvey_BackgroundRounded
     tiling=NoTile
     imageSize=<32 32>
     margin=<ltrb 0 0 0 0>
     outlineSettings {
      bUseBrushTransparency=false
      roundingType=HalfHeightRadius
      width=0
      color {colorUseRule=UseColor_Specified, specifiedColor=<rgba 0 0 0 0>}
      cornerRadii=<0 0 0 0>
     }
     tintColor {colorUseRule=UseColor_Specified, specifiedColor=<rgba 1 1 1 1>}
     uVRegion {bIsValid=false, max=<0 0>, min=<0 0>}
    }
    padding=<ltrb 12 12 12 12>
   }
   slot {class=VerticalBoxSlot, properties {horizontalAlignment=HAlign_Center}}
  }
  WBP_QRCode {
   widgetClassPath=/QRCodeGenerator/Widgets/WBP_QRCodeGenerator
   bIsVariable=true
   parent=Border_0
   widget=WBP_QRCode
   slot {class=BorderSlot, properties {padding=<ltrb 12 12 12 12>}}
  }
  Button_OpenLink {
   widgetClassPath=&WBP_UIKit_Button_Regular
   bIsVariable=true
   parent=VerticalBox_162
   widget=Button_OpenLink
   properties {text=Open Link in Browser}
   slot {class=VerticalBoxSlot, properties {horizontalAlignment=HAlign_Center, padding=<ltrb 0 32 0 0>}}
  }
 ]
}
