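# Player survey widget that shows a gamepad move-input icon framed by directional arrows as a navigation hint.
# PreConstruct uses the vertical and horizontal flags to toggle arrow pairs, with smaller arrows and tighter icon padding when both show.
# It binds the CommonActionWidget to the up/down move action whenever vertical arrows show, otherwise left/right.
# On PreConstruct and input method changes, the whole widget copies the action widget's visibility.
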WidgetBlueprint /FortniteCoreUI/UI/PlayerSurvey/Widgets/PlayerSurveyOrthogonalIndicator
graphs [
 EventGraph {nodes [
   "On Input Method Changed (InputActionWidget_Move)"!() -> (OutputDelegate, then, bUsingGamepad:bool) {is=title K2Node_ComponentBoundEvent}
   > UpdateGamepadControlsVisibility()
   Event PreConstruct()
   > SetVisibility(
    self=[Get Image_DownArrow() -> (Image_DownArrow:Image Object Reference), Get Image_UpArrow() -> (Image_UpArrow:Image Object Reference)]
    InVisibility=Select(Option 0=Collapsed, Option 1=SelfHitTestInvisible, Index=Get bShowVerticalArrows() -> (bShowVerticalArrows:bool))
   )
   > SetVisibility(
    self=[Get Image_LeftArrow() -> (Image_LeftArrow:Image Object Reference), Get Image_RightArrow() -> (Image_RightArrow:Image Object Reference)]
    InVisibility=Select(Option 0=Collapsed, Option 1=SelfHitTestInvisible, Index=Get bShowHorizontalArrows() -> (bShowHorizontalArrows:bool))
   )
   n1 = Get bShowVerticalArrows()
   n2 = Get bShowHorizontalArrows()
   > SetDesiredSizeOverride!(
    execute
    self:Image Object Reference=[Get Image_UpArrow(), Get Image_DownArrow(), Get Image_LeftArrow(), Get Image_RightArrow()]
    DesiredSize:Vector 2D Structure=Select(Option 0:Vector 2D Structure=(X=32,Y=32), Option 1:Vector 2D Structure=(X=24,Y=24), Index=AND Boolean(A=@n1, B=@n2)) -> (ReturnValue:Vector 2D Structure)
   ) -> (then)
   n3 = Select(Option 0:float=8, Option 1:float=4, Index=AND Boolean(A=Get bShowVerticalArrows(), B=Get bShowHorizontalArrows())) -> (ReturnValue:float)
   n4 = Select(Option 0=0, Option 1=@n3, Index=Get bShowVerticalArrows())
   n5 = Select(Option 0=0, Option 1=@n3, Index=Get bShowHorizontalArrows())
   > SetPadding!(
    execute
    self:Vertical Box Slot Object Reference=SlotAsVerticalBoxSlot(Widget=Get InputActionWidget_Move() -> (InputActionWidget_Move:Common Action Widget Object Reference))
    InPadding_Left:float=@n5
    InPadding_Top:float=@n4
    InPadding_Right:float=@n5
    InPadding_Bottom:float=@n4
   ) -> (then)
   > SetInputAction!(
    execute
    self:Common Action Widget Object Reference=Get InputActionWidget_Move()
    InputActionRow_DataTable:Data Table Object Reference=&InputActionDataTable
    InputActionRow_RowName:Name=Select(Option 0:Name=GenericMoveLeftRight, Option 1:Name=GenericMoveUpDown, Index=Get bShowVerticalArrows()) -> (ReturnValue:Name)
   ) -> (then)
   > UpdateGamepadControlsVisibility()
  ]}
 UpdateGamepadControlsVisibility {nodes [
   UpdateGamepadControlsVisibility() {is=title K2Node_FunctionEntry}
   > SetVisibility(InVisibility=GetVisibility!(self:Widget Object Reference=Get InputActionWidget_Move()) -> (ReturnValue:ESlateVisibility Enum=Visible))
  ]}
]
variables [
 bShowHorizontalArrows {type=boolean, default=true}
 bShowVerticalArrows {type=boolean, default=true}
]
widgets {
 info {inheritedWidgetCount=0, parentClass=&UserWidget, rootWidgetClass=&HorizontalBox, widgetCount=7}
 widgets [
  HorizontalBox_0 {widgetClassPath=&HorizontalBox, widget=HorizontalBox_0}
  Image_LeftArrow {
   widgetClassPath=&Image
   bIsVariable=true
   parent=HorizontalBox_0
   widget=Image_LeftArrow
   properties {
    brush {
     bIsDynamicallyLoaded=false
     drawAs=Image
     imageType=NoImage
     mirroring=NoMirror
     resourceName=None
     resourceObject=/FortniteCoreUI/Frontend/Showdown/Materials/M_UI_ArrowCircle_PlayerSurvey
     tiling=NoTile
     imageSize=<24 24>
     margin=<ltrb 0 0 0 0>
     outlineSettings {
      bUseBrushTransparency=false
      roundingType=HalfHeightRadius
      width=0
      color {colorUseRule=UseColor_Specified, specifiedColor=<rgba 0 0 0 0>}
      cornerRadii=<0 0 0 0>
     }
     tintColor {colorUseRule=UseColor_Specified, specifiedColor=<rgba 1 1 1 1>}
     uVRegion {bIsValid=false, max=<0 0>, min=<0 0>}
    }
    colorAndOpacity=<rgba 0 0.2706 0.7373 1>
    renderTransform {angle=-180, scale=<1 1>, shear=<0 0>, translation=<0 0>}
   }
   slot {class=HorizontalBoxSlot, properties {verticalAlignment=VAlign_Center}}
  }
  VerticalBox_0 {widgetClassPath=&VerticalBox, parent=HorizontalBox_0, widget=VerticalBox_0, slot {class=HorizontalBoxSlot}}
  Image_UpArrow {
   widgetClassPath=&Image
   bIsVariable=true
   parent=VerticalBox_0
   widget=Image_UpArrow
   properties {
    brush {
     bIsDynamicallyLoaded=false
     drawAs=Image
     imageType=NoImage
     mirroring=NoMirror
     resourceName=None
     resourceObject=&M_UI_ArrowCircle_PlayerSurvey
     tiling=NoTile
     imageSize=<24 24>
     margin=<ltrb 0 0 0 0>
     outlineSettings {
      bUseBrushTransparency=false
      roundingType=HalfHeightRadius
      width=0
      color {colorUseRule=UseColor_Specified, specifiedColor=<rgba 0 0 0 0>}
      cornerRadii=<0 0 0 0>
     }
     tintColor {colorUseRule=UseColor_Specified, specifiedColor=<rgba 1 1 1 1>}
     uVRegion {bIsValid=false, max=<0 0>, min=<0 0>}
    }
    colorAndOpacity=<rgba 0 0.2706 0.7373 1>
    renderTransform {angle=-90, scale=<1 1>, shear=<0 0>, translation=<0 0>}
   }
   slot {class=VerticalBoxSlot, properties {horizontalAlignment=HAlign_Center}}
  }
  InputActionWidget_Move {
   widgetClassPath=&CommonActionWidget
   bIsVariable=true
   parent=VerticalBox_0
   widget=InputActionWidget_Move
   properties {designTimeKey=Gamepad_LeftY}
   slot {class=VerticalBoxSlot, properties {horizontalAlignment=HAlign_Center, verticalAlignment=VAlign_Center}}
  }
  Image_DownArrow {
   widgetClassPath=&Image
   bIsVariable=true
   parent=VerticalBox_0
   widget=Image_DownArrow
   properties {
    brush {
     bIsDynamicallyLoaded=false
     drawAs=Image
     imageType=NoImage
     mirroring=NoMirror
     resourceName=None
     resourceObject=&M_UI_ArrowCircle_PlayerSurvey
     tiling=NoTile
     imageSize=<24 24>
     margin=<ltrb 0 0 0 0>
     outlineSettings {
      bUseBrushTransparency=false
      roundingType=HalfHeightRadius
      width=0
      color {colorUseRule=UseColor_Specified, specifiedColor=<rgba 0 0 0 0>}
      cornerRadii=<0 0 0 0>
     }
     tintColor {colorUseRule=UseColor_Specified, specifiedColor=<rgba 1 1 1 1>}
     uVRegion {bIsValid=false, max=<0 0>, min=<0 0>}
    }
    colorAndOpacity=<rgba 0 0.2706 0.7373 1>
    renderTransform {angle=90, scale=<1 1>, shear=<0 0>, translation=<0 0>}
   }
   slot {class=VerticalBoxSlot, properties {horizontalAlignment=HAlign_Center}}
  }
  Image_RightArrow {
   widgetClassPath=&Image
   bIsVariable=true
   parent=HorizontalBox_0
   widget=Image_RightArrow
   properties {
    brush {
     bIsDynamicallyLoaded=false
     drawAs=Image
     imageType=NoImage
     mirroring=NoMirror
     resourceName=None
     resourceObject=&M_UI_ArrowCircle_PlayerSurvey
     tiling=NoTile
     imageSize=<24 24>
     margin=<ltrb 0 0 0 0>
     outlineSettings {
      bUseBrushTransparency=false
      roundingType=HalfHeightRadius
      width=0
      color {colorUseRule=UseColor_Specified, specifiedColor=<rgba 0 0 0 0>}
      cornerRadii=<0 0 0 0>
     }
     tintColor {colorUseRule=UseColor_Specified, specifiedColor=<rgba 1 1 1 1>}
     uVRegion {bIsValid=false, max=<0 0>, min=<0 0>}
    }
    colorAndOpacity=<rgba 0 0.2706 0.7373 1>
   }
   slot {class=HorizontalBoxSlot, properties {verticalAlignment=VAlign_Center}}
  }
 ]
}
