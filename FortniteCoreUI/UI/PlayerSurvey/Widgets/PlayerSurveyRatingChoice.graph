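# Player survey widget for one rating choice: a square button with a centered value and a caption placed through a visual attachment.
# Hover, selection and reset events recolor the value from the current button text style; hover, reset and init update the select input icon, visible only while hovered.
# Initialization blanks the placeholder caption.
# GetCurrentButtonBrush picks a style brush from enabled, selected, pressed and hovered state.
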WidgetBlueprint /FortniteCoreUI/UI/PlayerSurvey/Widgets/PlayerSurveyRatingChoice
graphs [
 EventGraph {nodes [
   Event OnSurveyResetChoice() {is=title K2Node_Event}
   > UpdateTextStyle()
   > UpdateInputActionRenderOpacity()
   Event OnInitialized()
   > UpdateInputActionRenderOpacity()
   > "SetText (Text)"(self=Get Text_Label() -> (Text_Label:Common Text Object Reference))
   Event On Hovered()
   > UpdateTextStyle()
   > UpdateInputActionRenderOpacity()
   Event On Unhovered()
   > UpdateTextStyle()
   > UpdateInputActionRenderOpacity()
   Event On Selected()
   > UpdateTextStyle()
   Event On Deselected()
   > UpdateTextStyle()
  ]}
 GetCurrentButtonBrush {nodes [
   GetCurrentButtonBrush() {is=title K2Node_FunctionEntry}
   n1 = GetStyle!(self:Common Button Base Object Reference) -> (ReturnValue:Common Button Style Object Reference)
   n2 = GetSelected!(self:Common Button Base Object Reference) -> (ReturnValue:bool)
   n3 = Select(
    Option 0:Slate Brush Structure=Select(
     Option 0=GetNormalBaseBrush!(self:Common Button Style Object Reference=@n1) -> (Brush:Slate Brush Structure)
     Option 1=GetNormalHoveredBrush!(self:Common Button Style Object Reference=@n1) -> (Brush:Slate Brush Structure)
     Index=IsHovered()
    )
    Option 1:Slate Brush Structure=GetNormalPressedBrush!(self:Common Button Style Object Reference=@n1) -> (Brush:Slate Brush Structure)
    Index=IsPressed!(self:Common Button Base Object Reference) -> (ReturnValue:bool)
   ) -> (ReturnValue:Slate Brush Structure)
   n4 = Select(
    Option 0=Select(
     Option 0=GetSelectedBaseBrush!(self:Common Button Style Object Reference=@n1) -> (Brush:Slate Brush Structure)
     Option 1=GetSelectedHoveredBrush!(self:Common Button Style Object Reference=@n1) -> (Brush:Slate Brush Structure)
     Index=IsHovered()
    )
    Option 1=GetSelectedPressedBrush!(self:Common Button Style Object Reference=@n1) -> (Brush:Slate Brush Structure)
    Index=IsPressed()
   )
   > Return Node!(
    execute
    CurrentBrush:Slate Brush Structure=Select(
     Option 0=GetDisabledBrush!(self:Common Button Style Object Reference=@n1) -> (Brush:Slate Brush Structure)
     Option 1=Select(Option 0=@n3, Option 1=@n4, Index=@n2)
     Index=GetIsEnabled!(self:Widget Object Reference) -> (ReturnValue:bool)
    )
   )
  ]}
 UpdateInputActionRenderOpacity {nodes [
   UpdateInputActionRenderOpacity() {is=title K2Node_FunctionEntry}
   > SetRenderOpacity(
    self=Get InputActionWidget_Select() -> (InputActionWidget_Select:Common Action Widget Object Reference)
    InOpacity=Select(Option 0:double=0, Option 1:double=1, Index=IsHovered()) -> (ReturnValue:double)
   )
  ]}
 UpdateTextStyle {nodes [
   UpdateTextStyle() {is=title K2Node_FunctionEntry}
   > SetColorAndOpacity!(
    execute
    self=Get Text_RatingValue() -> (Text_RatingValue:Common Text Object Reference)
    InColorAndOpacity_SpecifiedColor:Linear Color Structure=GetColor!(self:Common Text Style Object Reference=GetCurrentTextStyle!(self:Common Button Base Object Reference) -> (ReturnValue:Common Text Style Object Reference)) -> (OutColor:Linear Color Structure)
    InColorAndOpacity_ColorUseRule:ESlateColorStylingMode Enum
   ) -> (then)
  ]}
]
variables []
widgets {
 info {inheritedWidgetCount=3, parentClass=/Script/FortniteUI.FortPlayerSurveyRatingChoiceBase, rootWidgetClass=&SizeBox, widgetCount=7}
 widgets [
  SizeBox_1 {
   widgetClassPath=&SizeBox
   widget=SizeBox_1
   properties {bOverride_HeightOverride=true, bOverride_WidthOverride=true, heightOverride=100, widthOverride=100}
  }
  Overlay_0 {widgetClassPath=&Overlay, parent=SizeBox_1, widget=Overlay_0, slot {class=SizeBoxSlot}}
  InputActionWidget_Select {
   widgetClassPath=&CommonActionWidget
   bIsVariable=true
   parent=Overlay_0
   widget=InputActionWidget_Select
   properties {inputActions [{dataTable=&InputActionDataTable, rowName=GenericForward}]}
   slot {class=OverlaySlot}
  }
  Text_RatingValue {
   widgetClassPath=&CommonTextBlock
   bInherited=true
   bIsVariable=true
   parent=Overlay_0
   widget=Text_RatingValue
   properties {
    justification=Center
    text=5
    textTransformPolicy=ToUpper
    wrapTextAt=400
    font {
     bForceMonospaced=false
     bMaterialIsStencil=false
     fontMaterial=null
     fontObject=&BurbankSmall
     letterSpacing=0
     monospacedWidth=1
     size=40
     skewAmount=0
     typefaceFontName=Black
     outlineSettings {
      bApplyOutlineToDropShadows=false
      bMiteredCorners=false
      bSeparateFillAlpha=false
      outlineMaterial=null
      outlineSize=0
      outlineColor=<rgba 0 0 0 1>
     }
    }
    shadowOffset=<0 0>
   }
   slot {class=OverlaySlot, properties {horizontalAlignment=HAlign_Center, verticalAlignment=VAlign_Center}}
  }
  FortVisualAttachment_0 {
   widgetClassPath=&FortVisualAttachment
   parent=Overlay_0
   widget=FortVisualAttachment_0
   properties {contentAnchor=<0.5 0>}
   slot {class=OverlaySlot, properties {horizontalAlignment=HAlign_Center, verticalAlignment=VAlign_Bottom}}
  }
  Text_Label {
   widgetClassPath=&CommonTextBlock
   bInherited=true
   bIsVariable=true
   parent=FortVisualAttachment_0
   widget=Text_Label
   properties {
    justification=Center
    text=Too Strong
    wrapTextAt=256
    font {
     bForceMonospaced=false
     bMaterialIsStencil=false
     fontMaterial=null
     fontObject=&BurbankSmall
     letterSpacing=0
     monospacedWidth=1
     size=15
     skewAmount=0
     typefaceFontName=Black
     outlineSettings {
      bApplyOutlineToDropShadows=false
      bMiteredCorners=false
      bSeparateFillAlpha=false
      outlineMaterial=null
      outlineSize=0
      outlineColor=<rgba 0 0 0 1>
     }
    }
    shadowOffset=<0 0>
   }
   slot {class=SizeBoxSlot, properties {padding=<ltrb 0 24 0 0>}}
  }
  InputActionWidget {widgetClassPath=&CommonActionWidget, bInherited=true, widget=null}
 ]
}
