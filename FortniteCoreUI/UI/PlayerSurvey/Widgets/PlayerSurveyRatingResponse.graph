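# Widget for a rating question in the player survey, built on FortPlayerSurveyRatingResponseBase.
# It stacks a "Select a rating" prompt, a dynamic entry box of PlayerSurveyRatingChoice widgets and an indicator with vertical arrows hidden.
# When responses are set, it adjusts the entry box's horizontal spacing by entry count, spreading three or fewer choices much wider than larger sets.
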WidgetBlueprint /FortniteCoreUI/UI/PlayerSurvey/Widgets/PlayerSurveyRatingResponse
graphs [EventGraph {nodes [
   Event On Responses Set() {is=title K2Node_Event}
   n1 = GetNumEntries!(self:Dynamic Entry Box Base Object Reference=Get DynamicEntryBox_Responses() -> (DynamicEntryBox_Responses:Dynamic Entry Box Object Reference)) -> (ReturnValue:int=0)
   > SetEntrySpacing!(
    execute
    self:Dynamic Entry Box Base Object Reference=Get DynamicEntryBox_Responses()
    InEntrySpacing_X:double=SelectFloat(A=200, B=50, bPickA="integer <= integer"!(A:int=@n1, B:int=3) -> (ReturnValue:bool) {is=title K2Node_CallFunction})
    InEntrySpacing_Y:double=0
   ) -> (then)
  ]}]
variables []
widgets {
 info {
  inheritedWidgetCount=1
  parentClass=/Script/FortniteUI.FortPlayerSurveyRatingResponseBase
  rootWidgetClass=&VerticalBox
  widgetCount=4
 }
 widgets [
  VerticalBox_0 {widgetClassPath=&VerticalBox, widget=VerticalBox_0}
  Text_Instruction {
   widgetClassPath=&CommonTextBlock
   parent=VerticalBox_0
   widget=Text_Instruction
   properties {
    justification=Center
    text=Select a rating
    colorAndOpacity {colorUseRule=UseColor_Specified, specifiedColor=<rgba 0.5776 0.5776 0.5776 1>}
    font {
     bForceMonospaced=false
     bMaterialIsStencil=false
     fontMaterial=null
     fontObject=&BurbankSmall
     letterSpacing=0
     monospacedWidth=1
     size=20
     skewAmount=0
     typefaceFontName=Black
     outlineSettings {
      bApplyOutlineToDropShadows=false
      bMiteredCorners=false
      bSeparateFillAlpha=false
      outlineMaterial=null
      outlineSize=0
      outlineColor=<rgba 0 0 0 1>
     }
    }
    shadowOffset=<0 0>
   }
   slot {class=VerticalBoxSlot, properties {horizontalAlignment=HAlign_Center, padding=<ltrb 0 0 0 56>}}
  }
  DynamicEntryBox_Responses {
   widgetClassPath=&DynamicEntryBox
   bInherited=true
   bIsVariable=true
   parent=VerticalBox_0
   widget=DynamicEntryBox_Responses
   properties {
    entryWidgetClass=&/PlayerSurveyRatingChoice.PlayerSurveyRatingChoice_C
    numDesignerPreviewEntries=5
    entrySpacing=<56 0>
   }
   slot {
    class=VerticalBoxSlot
    properties {horizontalAlignment=HAlign_Center, verticalAlignment=VAlign_Top, padding=<ltrb 0 0 0 56>, size {sizeRule=Fill, value=1}}
   }
  }
  PlayerSurveyHorizontalIndicator {
   widgetClassPath=&/PlayerSurveyOrthogonalIndicator
   parent=VerticalBox_0
   widget=PlayerSurveyHorizontalIndicator
   properties {bShowVerticalArrows=false}
   slot {class=VerticalBoxSlot, properties {horizontalAlignment=HAlign_Center, verticalAlignment=VAlign_Top, padding=<ltrb 0 56 0 0>}}
  }
 ]
}
