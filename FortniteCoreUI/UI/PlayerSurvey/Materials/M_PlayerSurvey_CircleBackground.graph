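# M_PlayerSurvey_CircleBackground is a translucent UI-domain material, named as the circle background for the player survey UI.
# Its graph is a single MF_UI_Circle function call, with constant inputs setting the circle's size and outline alignment.
# It exposes no parameters of its own, and the graph as shown has no connection from the function to the material outputs.
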Material /FortniteCoreUI/UI/PlayerSurvey/Materials/M_PlayerSurvey_CircleBackground
graph {
 properties {
  bAutomaticallySetUsageInEditor=false
  bUsedWithStaticMesh=false
  blendMode=BLEND_Translucent
  materialDomain=MD_UI
  usedShadingModels=Default Lit
 }
 nodes [MaterialFunctionCall(Size=Constant() {r=1}, OutlineAlignment=Constant() {r=-1}) {materialFunction=&MF_UI_Circle}]
}
