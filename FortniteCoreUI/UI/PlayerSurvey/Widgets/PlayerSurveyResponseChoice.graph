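# Player survey answer button that shows a response beside a checkbox or radio image chosen by choice type and selection.
# Button state events refresh that image, text and checkbox tints, and the input-prompt circle color taken from the current style brush.
# The select prompt appears only on hover, switching its icon material to "Use Dark" when that brush tint's weighted luminance is light.
# Text style changes shrink the response font if the text grows too tall.
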WidgetBlueprint /FortniteCoreUI/UI/PlayerSurvey/Widgets/PlayerSurveyResponseChoice
graphs [
 EventGraph {nodes [
   "On Input Icon Updated (InputActionWidget_Select)"() {is=title K2Node_ComponentBoundEvent}
   > UpdateInputActionColor()
   Event OnSurveyResetChoice() {is=title K2Node_Event}
   > UpdateChoiceImage()
   > UpdateButtonColor()
   > UpdateInputActionRenderOpacity()
   Event On Hovered()
   > UpdateButtonColor()
   > UpdateInputActionRenderOpacity()
   Event On Unhovered()
   > UpdateButtonColor()
   > UpdateInputActionRenderOpacity()
   Event On Selected()
   > UpdateChoiceImage()
   > UpdateButtonColor()
   n1 = Event SetChoiceType!() -> (OutputDelegate, then, Type:EFortPlayerSurveyResponseChoiceType Enum=Checkbox) {is=title K2Node_Event}
   > Set ChoiceType(ChoiceType:EFortPlayerSurveyResponseChoiceType Enum=@n1.Type) -> (Output_Get:EFortPlayerSurveyResponseChoiceType Enum=Checkbox)
   > UpdateChoiceImage()
   > UpdateButtonColor()
   Event On Deselected()
   > UpdateChoiceImage()
   > UpdateButtonColor()
   Event OnInitialized()
   > UpdateButtonColor()
   > UpdateInputActionRenderOpacity()
   Event OnCurrentTextStyleChanged() {is=title K2Node_Event}
   > Update Text Scale()
  ]}
 GetCurrentButtonBrush {nodes [
   GetCurrentButtonBrush() {is=title K2Node_FunctionEntry}
   n1 = GetStyle!(self:Common Button Base Object Reference) -> (ReturnValue:Common Button Style Object Reference)
   n2 = GetSelected!(self:Common Button Base Object Reference) -> (ReturnValue:bool)
   n3 = Select(
    Option 0:Slate Brush Structure=Select(
     Option 0=GetNormalBaseBrush!(self:Common Button Style Object Reference=@n1) -> (Brush:Slate Brush Structure)
     Option 1=GetNormalHoveredBrush!(self:Common Button Style Object Reference=@n1) -> (Brush:Slate Brush Structure)
     Index=IsHovered()
    )
    Option 1:Slate Brush Structure=GetNormalPressedBrush!(self:Common Button Style Object Reference=@n1) -> (Brush:Slate Brush Structure)
    Index=IsPressed!(self:Common Button Base Object Reference) -> (ReturnValue:bool)
   ) -> (ReturnValue:Slate Brush Structure)
   n4 = Select(
    Option 0=Select(
     Option 0=GetSelectedBaseBrush!(self:Common Button Style Object Reference=@n1) -> (Brush:Slate Brush Structure)
     Option 1=GetSelectedHoveredBrush!(self:Common Button Style Object Reference=@n1) -> (Brush:Slate Brush Structure)
     Index=IsHovered()
    )
    Option 1=GetSelectedPressedBrush!(self:Common Button Style Object Reference=@n1) -> (Brush:Slate Brush Structure)
    Index=IsPressed()
   )
   > Return Node!(
    execute
    CurrentBrush:Slate Brush Structure=Select(
     Option 0=GetDisabledBrush!(self:Common Button Style Object Reference=@n1) -> (Brush:Slate Brush Structure)
     Option 1=Select(Option 0=@n3, Option 1=@n4, Index=@n2)
     Index=GetIsEnabled!(self:Widget Object Reference) -> (ReturnValue:bool)
    )
   )
  ]}
 IsButtonColorLight {nodes [
   IsButtonColorLight() {is=title K2Node_FunctionEntry}
   n1 = GetCurrentButtonBrush!() -> (
    CurrentBrush_TintColor_SpecifiedColor_R:float=~
    CurrentBrush_TintColor_SpecifiedColor_G:float=~
    CurrentBrush_TintColor_SpecifiedColor_B:float=~
    CurrentBrush_TintColor_SpecifiedColor_A:float=~
    CurrentBrush_TintColor_ColorUseRule:ESlateColorStylingMode Enum
    CurrentBrush_DrawAs:ESlateBrushDrawType Enum
    CurrentBrush_Tiling:ESlateBrushTileType Enum
    CurrentBrush_Mirroring:ESlateBrushMirrorType Enum
    CurrentBrush_ImageSize:Vector 2D Structure
    CurrentBrush_Margin:Margin Structure
    CurrentBrush_ResourceObject:Object Reference
    CurrentBrush_OutlineSettings:Slate Brush Outline Settings Structure
   ) {is=title K2Node_CallFunction}
   n2 = float * float(A=@n1, B=0.2126) {is=title K2Node_PromotableOperator}
   n3 = float * float(A=@n1.CurrentBrush_TintColor_SpecifiedColor_G, B=0.7152)
   n4 = float * float(A=@n1.CurrentBrush_TintColor_SpecifiedColor_B, B=0.0722)
   > Return Node!(execute, bIsLight:bool="float >= float"(A=float + float!(A=@n2, B=@n3, C:double=@n4) -> (ReturnValue), B=0.5))
  ]}
 Update Text Scale {nodes [
   Update Text Scale() {is=title K2Node_FunctionEntry}
   > ForceLayoutPrepass(self:Widget Object Reference=Get Text_Response() -> (Text_Response:Common Text Object Reference))
   n1 = Get Text_Response()
   > Branch(Condition="float > float"(A=GetDesiredSize!(self:Widget Object Reference=@n1) -> (ReturnValue_X:double, ReturnValue_Y:double).ReturnValue_Y, B=70) {is=title K2Node_CallFunction})
   n2 = Get!(self:Text Object Reference=Get Text_Response()) -> (
    Font_FontObject:Object Reference
    Font_FontMaterial:Object Reference
    Font_OutlineSettings:Font Outline Settings Structure
    Font_TypefaceFontName:Name
    Font_Size:float=~
    Font_LetterSpacing:int
    Font_SkewAmount:float=~
    Font_bForceMonospaced:bool=~
    Font_bMaterialIsStencil:bool=~
    Font_MonospacedWidth:float=~
   )
   > SetFont!(
    execute
    self=Get Text_Response()
    InFontInfo_FontObject:Object Reference=@n2
    InFontInfo_FontMaterial:Object Reference=@n2.Font_FontMaterial
    InFontInfo_OutlineSettings:Font Outline Settings Structure=@n2.Font_OutlineSettings
    InFontInfo_TypefaceFontName:Name=@n2.Font_TypefaceFontName
    InFontInfo_Size:float="To Float (Integer)"(InInt=MakeLiteralInt!(Value:int=15) -> (ReturnValue:int=0))
    InFontInfo_LetterSpacing:int=@n2.Font_LetterSpacing
    InFontInfo_SkewAmount:float=0
    InFontInfo_bForceMonospaced:bool=false
    InFontInfo_bMaterialIsStencil:bool=false
    InFontInfo_MonospacedWidth:float=1
   ) -> (then)
  ]}
 UpdateButtonColor {nodes [
   UpdateButtonColor() {is=title K2Node_FunctionEntry}
   > SetBrushColor(
    self=Get Border_InputActionCircle() -> (Border_InputActionCircle:Border Object Reference)
    InBrushColor=GetCurrentButtonBrush!() -> (
     CurrentBrush_TintColor_SpecifiedColor:Linear Color Structure
     CurrentBrush_TintColor_ColorUseRule
     CurrentBrush_DrawAs
     CurrentBrush_Tiling
     CurrentBrush_Mirroring
     CurrentBrush_ImageSize
     CurrentBrush_Margin
     CurrentBrush_ResourceObject
     CurrentBrush_OutlineSettings
    )
   )
   n1 = GetCurrentTextStyle!(self:Common Button Base Object Reference) -> (ReturnValue:Common Text Style Object Reference)
   n2 = GetSelected()
   n3 = "Equal (Enum)"(
    A:EFortPlayerSurveyResponseChoiceType Enum=Get ChoiceType() -> (ChoiceType:EFortPlayerSurveyResponseChoiceType Enum=Checkbox)
    B:EFortPlayerSurveyResponseChoiceType Enum=Checkbox
   )
   > SetBrushTintColor!(
    execute
    self:Image Object Reference=Get Image_Checkbox() -> (Image_Checkbox:Image Object Reference)
    TintColor_SpecifiedColor:Linear Color Structure=Select(
     Option 0:Linear Color Structure=GetColor!(self:Common Text Style Object Reference=@n1) -> (OutColor:Linear Color Structure)
     Option 1:Linear Color Structure=(R=1,G=1,B=1,A=1)
     Index=AND Boolean(A=@n3, B=@n2)
    ) -> (ReturnValue:Linear Color Structure)
    TintColor_ColorUseRule:ESlateColorStylingMode Enum
   ) -> (then)
   > Set!(
    execute
    ColorAndOpacity_SpecifiedColor:Linear Color Structure=GetColor(self=GetCurrentTextStyle())
    ColorAndOpacity_ColorUseRule:ESlateColorStylingMode Enum
    self:Text Object Reference=Get Text_Response()
   ) -> (then, Output_Get:Slate Color Structure)
   > UpdateInputActionColor()
  ]}
 UpdateChoiceImage {nodes [
   UpdateChoiceImage() {is=title K2Node_FunctionEntry}
   n1 = GetSelected()
   n2 = Get Brush_RadioUnchecked() -> (Brush_RadioUnchecked:Slate Brush Structure)
   n3 = Get Brush_RadioChecked() -> (Brush_RadioChecked:Slate Brush Structure)
   n4 = GetSelected()
   n5 = Get Brush_CheckboxUnchecked() -> (Brush_CheckboxUnchecked:Slate Brush Structure)
   n6 = Get Brush_CheckboxChecked() -> (Brush_CheckboxChecked:Slate Brush Structure)
   > SetBrush(
    self=Get Image_Checkbox()
    InBrush=Select!(
     Checkbox:"Slate Brush Structure (by ref)"=Select!(Option 0:"Slate Brush Structure (by ref)"=@n5, Option 1:"Slate Brush Structure (by ref)"=@n6, Index:bool=@n4) -> (ReturnValue)
     Radio:"Slate Brush Structure (by ref)"=Select(Option 0=@n2, Option 1=@n3, Index=@n1)
     Index:EFortPlayerSurveyResponseChoiceType Enum=Get ChoiceType() -> (ChoiceType=Checkbox)
    ) -> (ReturnValue:"Slate Brush Structure (by ref)")
   )
  ]}
 UpdateInputActionColor {nodes [
   UpdateInputActionColor() {is=title K2Node_FunctionEntry}
   > n1 = GetIconDynamicMaterial!(
    execute
    self:Common Action Widget Object Reference=Get InputActionWidget_Select() -> (InputActionWidget_Select:Common Action Widget Object Reference)
   ) -> (then, ReturnValue:Material Instance Dynamic Object Reference)
   > SetScalarParameterValue(
    self=@n1
    ParameterName=Use Dark
    Value=Select(Option 0:float=0, Option 1:float=1, Index=IsButtonColorLight!() -> (bIsLight:bool) {is=title K2Node_CallFunction}) -> (ReturnValue:float)
   )
  ]}
 UpdateInputActionRenderOpacity {nodes [
   UpdateInputActionRenderOpacity() {is=title K2Node_FunctionEntry}
   n1 = Select(Option 0:double=0, Option 1:double=1, Index=IsHovered()) -> (ReturnValue:double)
   > SetRenderOpacity(self=Get Border_InputActionCircle(), InOpacity=@n1)
   > SetRenderOpacity(self=Get InputActionWidget_Select(), InOpacity=@n1)
  ]}
]
variables [
 Brush_CheckboxChecked {
  type=SlateBrush
  default {
   bIsDynamicallyLoaded=false
   drawAs=Image
   imageType=NoImage
   mirroring=NoMirror
   resourceName=None
   resourceObject=/FortUILibrary/Foundation/Art_ToolKit/AngledBoxes/M_UI_AngledCheckBox
   tiling=NoTile
   imageSize=<32 32>
   margin=<ltrb 0 0 0 0>
   outlineSettings {
    bUseBrushTransparency=false
    roundingType=HalfHeightRadius
    width=0
    color {colorUseRule=UseColor_Specified, specifiedColor=<rgba 0 0 0 0>}
    cornerRadii=<0 0 0 1>
   }
   tintColor {colorUseRule=UseColor_Specified, specifiedColor=<rgba 1 1 1 1>}
   uVRegion {bIsValid=false, max=<0 0>, min=<0 0>}
  }
 }
 Brush_CheckboxUnchecked {
  type=SlateBrush
  default {
   bIsDynamicallyLoaded=false
   drawAs=Image
   imageType=NoImage
   mirroring=NoMirror
   resourceName=None
   resourceObject=/FortUILibrary/Foundation/Art_ToolKit/AngledBoxes/M_UI_AngledCheckBox_Empty
   tiling=NoTile
   imageSize=<32 32>
   margin=<ltrb 0 0 0 0>
   outlineSettings {
    bUseBrushTransparency=false
    roundingType=HalfHeightRadius
    width=0
    color {colorUseRule=UseColor_Specified, specifiedColor=<rgba 0 0 0 0>}
    cornerRadii=<0 0 0 1>
   }
   tintColor {colorUseRule=UseColor_Specified, specifiedColor=<rgba 1 1 1 1>}
   uVRegion {bIsValid=false, max=<0 0>, min=<0 0>}
  }
 }
 Brush_RadioChecked {
  type=SlateBrush
  default {
   bIsDynamicallyLoaded=false
   drawAs=Image
   imageType=FullColor
   mirroring=NoMirror
   resourceName=None
   resourceObject=/FortniteCoreUI/UI/PlayerSurvey/Materials/MI_PlayerSurvey_RadioOptionCircle_Selected
   tiling=NoTile
   imageSize=<32 32>
   margin=<ltrb 0 0 0 0>
   outlineSettings {
    bUseBrushTransparency=false
    roundingType=HalfHeightRadius
    width=0
    color {colorUseRule=UseColor_Specified, specifiedColor=<rgba 0 0 0 0>}
    cornerRadii=<0 0 0 1>
   }
   tintColor {colorUseRule=UseColor_Specified, specifiedColor=<rgba 1 1 1 1>}
   uVRegion {bIsValid=false, max=<0 0>, min=<0 0>}
  }
 }
 Brush_RadioUnchecked {
  type=SlateBrush
  default {
   bIsDynamicallyLoaded=false
   drawAs=Image
   imageType=FullColor
   mirroring=NoMirror
   resourceName=None
   resourceObject=/FortniteCoreUI/UI/PlayerSurvey/Materials/MI_PlayerSurvey_RadioOptionCircle_Unselected
   tiling=NoTile
   imageSize=<32 32>
   margin=<ltrb 0 0 0 0>
   outlineSettings {
    bUseBrushTransparency=false
    roundingType=HalfHeightRadius
    width=0
    color {colorUseRule=UseColor_Specified, specifiedColor=<rgba 0 0 0 0>}
    cornerRadii=<0 0 0 1>
   }
   tintColor {colorUseRule=UseColor_Specified, specifiedColor=<rgba 1 1 1 1>}
   uVRegion {bIsValid=false, max=<0 0>, min=<0 0>}
  }
 }
 ChoiceType {type=EFortPlayerSurveyResponseChoiceType, default=Radio}
]
widgets {
 info {
  inheritedWidgetCount=2
  parentClass=/Script/FortniteUI.FortPlayerSurveyResponseChoiceBase
  rootWidgetClass=&SizeBox
  widgetCount=10
 }
 widgets [
  SizeBox_1 {
   widgetClassPath=&SizeBox
   widget=SizeBox_1
   properties {bOverride_HeightOverride=true, bOverride_WidthOverride=true, heightOverride=96, widthOverride=500}
  }
  Overlay_1 {widgetClassPath=&Overlay, parent=SizeBox_1, widget=Overlay_1, slot {class=SizeBoxSlot}}
  FortVisualAttachment_0 {
   widgetClassPath=&FortVisualAttachment
   parent=Overlay_1
   widget=FortVisualAttachment_0
   properties {contentAnchor=<0.5 0.5>}
   slot {class=OverlaySlot, properties {horizontalAlignment=HAlign_Center}}
  }
  Border_InputActionCircle {
   widgetClassPath=&Border
   bIsVariable=true
   parent=FortVisualAttachment_0
   widget=Border_InputActionCircle
   properties {
    background {
     bIsDynamicallyLoaded=false
     drawAs=Image
     imageType=FullColor
     mirroring=NoMirror
     resourceName=None
     resourceObject=/FortniteCoreUI/UI/PlayerSurvey/Materials/M_PlayerSurvey_CircleBackground
     tiling=NoTile
     imageSize=<64 64>
     margin=<ltrb 0 0 0 0>
     outlineSettings {
      bUseBrushTransparency=false
      roundingType=HalfHeightRadius
      width=0
      color {colorUseRule=UseColor_Specified, specifiedColor=<rgba 0 0 0 0>}
      cornerRadii=<0 0 0 0>
     }
     tintColor {colorUseRule=UseColor_Specified, specifiedColor=<rgba 1 1 1 1>}
     uVRegion {bIsValid=false, max=<0 0>, min=<0 0>}
    }
    brushColor=<rgba 0 0.1412 1 1>
    padding=<ltrb 4 4 4 4>
   }
   slot {class=SizeBoxSlot}
  }
  InputActionWidget_Select {
   widgetClassPath=&CommonActionWidget
   bIsVariable=true
   parent=Border_InputActionCircle
   widget=InputActionWidget_Select
   properties {designTimeKey=Gamepad_FaceButton_Bottom, inputActions [{dataTable=&InputActionDataTable, rowName=GenericForward}]}
   slot {class=BorderSlot, properties {padding=<ltrb 4 4 4 4>}}
  }
  HorizontalBox_0 {
   widgetClassPath=&HorizontalBox
   parent=Overlay_1
   widget=HorizontalBox_0
   slot {class=OverlaySlot, properties {verticalAlignment=VAlign_Center, padding=<ltrb 24 0 24 0>}}
  }
  Image_Checkbox {
   widgetClassPath=&Image
   bIsVariable=true
   parent=HorizontalBox_0
   widget=Image_Checkbox
   properties {brush {
     bIsDynamicallyLoaded=false
     drawAs=Image
     imageType=FullColor
     mirroring=NoMirror
     resourceName=None
     resourceObject=&MI_PlayerSurvey_RadioOptionCircle_Unselected
     tiling=NoTile
     imageSize=<30 30>
     margin=<ltrb 0 0 0 0>
     outlineSettings {
      bUseBrushTransparency=false
      roundingType=HalfHeightRadius
      width=0
      color {colorUseRule=UseColor_Specified, specifiedColor=<rgba 0 0 0 0>}
      cornerRadii=<0 0 0 0>
     }
     tintColor {colorUseRule=UseColor_Specified, specifiedColor=<rgba 0.2157 0.6588 1 1>}
     uVRegion {bIsValid=false, max=<0 0>, min=<0 0>}
    }}
   slot {class=HorizontalBoxSlot, properties {horizontalAlignment=HAlign_Center, verticalAlignment=VAlign_Center}}
  }
  Spacer_107 {widgetClassPath=&Spacer, parent=HorizontalBox_0, widget=Spacer_107, properties {size=<10 0>}, slot {class=HorizontalBoxSlot}}
  Text_Response {
   widgetClassPath=&CommonTextBlock
   bInherited=true
   bIsVariable=true
   parent=HorizontalBox_0
   widget=Text_Response
   properties {
    text=Charge Shotgun
    textTransformPolicy=ToUpper
    wrapTextAt=400
    colorAndOpacity {colorUseRule=UseColor_Specified, specifiedColor=<rgba 0.2159 0.6584 1 1>}
    font {
     bForceMonospaced=false
     bMaterialIsStencil=false
     fontMaterial=null
     fontObject=&BurbankSmall
     letterSpacing=0
     monospacedWidth=1
     size=20
     skewAmount=0
     typefaceFontName=Black
     outlineSettings {
      bApplyOutlineToDropShadows=false
      bMiteredCorners=false
      bSeparateFillAlpha=false
      outlineMaterial=null
      outlineSize=0
      outlineColor=<rgba 0 0 0 1>
     }
    }
    shadowOffset=<0 0>
   }
   slot {class=HorizontalBoxSlot, properties {horizontalAlignment=HAlign_Left, verticalAlignment=VAlign_Center}}
  }
  InputActionWidget {widgetClassPath=&CommonActionWidget, bInherited=true, widget=null}
 ]
}
