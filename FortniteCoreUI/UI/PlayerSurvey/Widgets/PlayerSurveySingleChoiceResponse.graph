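# Player survey widget that lays out single-choice response entries in a wrapping entry box, based on FortPlayerSurveySingleChoiceResponseBase.
# When SetNumberOfResponses fires, it adds a column for every three responses, up to three columns, and resizes the SizeBox so the entries wrap into that grid.
# A vertical arrow indicator beside the entries shows only in the single-column layout and collapses otherwise.
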WidgetBlueprint /FortniteCoreUI/UI/PlayerSurvey/Widgets/PlayerSurveySingleChoiceResponse
graphs [EventGraph {nodes [
   n1 = Event SetNumberOfResponses!() -> (OutputDelegate, then, NumberOfResponses:int=0) {is=title K2Node_Event}
   n2 = "Min (Integer)"(
    A=int / int!(A:int=int - int(A=@n1.NumberOfResponses, B=1) {is=title K2Node_CallFunction}, B:int=3) -> (ReturnValue:int=0) {is=title K2Node_CallFunction}
    B=2
   ) -> (ReturnValue=0)
   n3 = Get AdditionalColumnWidth() -> (AdditionalColumnWidth:double)
   > SetWidthOverride(
    self=Get SizeBox_Responses() -> (SizeBox_Responses:Size Box Object Reference)
    InWidthOverride=float + float(A=Get BaseColumnWidth() -> (BaseColumnWidth:double), B=int * float!(A:int=@n2, B:double=@n3) -> (ReturnValue:double)) {is=title K2Node_CommutativeAssociativeBinaryOperator}
   )
   > SetVisibility(
    self=Get Overlay_VerticalIndicator() -> (Overlay_VerticalIndicator:Overlay Object Reference)
    InVisibility=Select(Option 0=SelfHitTestInvisible, Option 1=Collapsed, Index="integer > integer"(A=@n2, B=0) {is=title K2Node_CallFunction})
   )
  ]}]
variables [
 AdditionalColumnWidth {type=number, default=542}
 BaseColumnWidth {type=number, default=512}
]
widgets {
 info {
  inheritedWidgetCount=1
  parentClass=/Script/FortniteUI.FortPlayerSurveySingleChoiceResponseBase
  rootWidgetClass=&VerticalBox
  widgetCount=6
 }
 widgets [
  VerticalBox_0 {widgetClassPath=&VerticalBox, widget=VerticalBox_0}
  HorizontalBox_0 {widgetClassPath=&HorizontalBox, parent=VerticalBox_0, widget=HorizontalBox_0, slot {class=VerticalBoxSlot}}
  SizeBox_Responses {
   widgetClassPath=&SizeBox
   bIsVariable=true
   parent=HorizontalBox_0
   widget=SizeBox_Responses
   properties {bOverride_WidthOverride=true, widthOverride=1596}
   slot {class=HorizontalBoxSlot}
  }
  DynamicEntryBox_Responses {
   widgetClassPath=&DynamicEntryBox
   bInherited=true
   bIsVariable=true
   parent=SizeBox_Responses
   widget=DynamicEntryBox_Responses
   properties {
    entryBoxType=Wrap
    entryWidgetClass=&/PlayerSurveyResponseChoice.PlayerSurveyResponseChoice_C
    numDesignerPreviewEntries=9
    entrySpacing=<24 24>
   }
   slot {class=SizeBoxSlot}
  }
  Overlay_VerticalIndicator {
   widgetClassPath=&Overlay
   bIsVariable=true
   parent=HorizontalBox_0
   widget=Overlay_VerticalIndicator
   slot {class=HorizontalBoxSlot, properties {verticalAlignment=VAlign_Center}}
  }
  PlayerSurveyVerticalIndicator {
   widgetClassPath=&/PlayerSurveyOrthogonalIndicator
   parent=Overlay_VerticalIndicator
   widget=PlayerSurveyVerticalIndicator
   properties {bShowHorizontalArrows=false}
   slot {class=OverlaySlot, properties {padding=<ltrb 48 0 0 0>}}
  }
 ]
}
