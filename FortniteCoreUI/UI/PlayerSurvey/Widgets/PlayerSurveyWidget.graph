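# Full-screen player survey widget that shows an optional feedback question over a dimmed scrim, built on FortPlayerSurveyWidgetBase.
# A safe-zone column stacks title, rich-text question, response slot, Skip and switchable Next/Submit buttons, opt-out note and question count.
# GetResponseWidget and SetResponseWidget read or replace the border's single child, a FortPlayerSurveyResponseWidget, making the answer UI swappable.
# The event graph is empty, leaving survey flow outside this Blueprint.
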WidgetBlueprint /FortniteCoreUI/UI/PlayerSurvey/Widgets/PlayerSurveyWidget
graphs [
 EventGraph {links [], nodes []}
 GetResponseWidget {nodes [
   GetResponseWidget() {is=title K2Node_FunctionEntry}
   > n1 = Cast To FortPlayerSurveyResponseWidget!(
    execute
    Object=GetChildAt!(self:Panel Widget Object Reference=Get Border_ResponseSlot() -> (Border_ResponseSlot:Border Object Reference), Index:int=0) -> (ReturnValue:Widget Object Reference)
   ) -> (then, CastFailed, AsFort Player Survey Response Widget:Fort Player Survey Response Widget Object Reference) {is=title K2Node_DynamicCast}
   > Return Node!(execute, ReturnValue:Fort Player Survey Response Widget Object Reference=@n1)
   Return Node(execute=@n1.CastFailed)
  ]}
 SetResponseWidget {nodes [
   SetResponseWidget!() -> (then, NewResponseWidget:Fort Player Survey Response Widget Object Reference) {is=title K2Node_FunctionEntry}
   > ClearChildren(self=Get Border_ResponseSlot())
   > AddChild(self=Get Border_ResponseSlot(), Content=Get NewResponseWidget() -> (NewResponseWidget:Fort Player Survey Response Widget Object Reference))
  ]}
]
variables []
widgets {
 info {inheritedWidgetCount=7, parentClass=/Script/FortniteUI.FortPlayerSurveyWidgetBase, rootWidgetClass=&Overlay, widgetCount=15}
 widgets [
  O_Container {widgetClassPath=&Overlay, widget=O_Container}
  WBP_UIKit_Scrim {
   widgetClassPath=/FortUILibrary/Systems/Dialog/WBP_UIKit_Scrim
   parent=O_Container
   widget=WBP_UIKit_Scrim
   slot {class=OverlaySlot, properties {horizontalAlignment=HAlign_Fill, verticalAlignment=VAlign_Fill}}
  }
  Overlay {
   widgetClassPath=&Image
   parent=O_Container
   widget=Overlay
   properties {
    visibility=HitTestInvisible
    brush {
     bIsDynamicallyLoaded=false
     drawAs=Image
     imageType=NoImage
     mirroring=NoMirror
     resourceName=None
     resourceObject=null
     tiling=NoTile
     imageSize=<32 32>
     margin=<ltrb 0 0 0 0>
     outlineSettings {
      bUseBrushTransparency=false
      roundingType=HalfHeightRadius
      width=0
      color {colorUseRule=UseColor_Specified, specifiedColor=<rgba 0 0 0 0>}
      cornerRadii=<0 0 0 0>
     }
     tintColor {colorUseRule=UseColor_Specified, specifiedColor=<rgba 0 0 0 0.5>}
     uVRegion {bIsValid=false, max=<0 0>, min=<0 0>}
    }
   }
   slot {class=OverlaySlot, properties {horizontalAlignment=HAlign_Fill, verticalAlignment=VAlign_Fill}}
  }
  SafeZone {
   widgetClassPath=&SafeZone
   parent=O_Container
   widget=SafeZone
   slot {class=OverlaySlot, properties {horizontalAlignment=HAlign_Fill, verticalAlignment=VAlign_Fill}}
  }
  VerticalBox_199 {
   widgetClassPath=&VerticalBox
   parent=SafeZone
   widget=VerticalBox_199
   slot {class=SafeZoneSlot, properties {padding=<ltrb 80 96 80 64>}}
  }
  Text_Title {
   widgetClassPath=&CommonTextBlock
   bInherited=true
   bIsVariable=true
   parent=VerticalBox_199
   widget=Text_Title
   properties {
    justification=Center
    text="[PH] We want your feedback!"
    textTransformPolicy=ToUpper
    wrapTextAt=1280
    font {
     bForceMonospaced=false
     bMaterialIsStencil=false
     fontMaterial=null
     fontObject=&Burbank
     letterSpacing=0
     monospacedWidth=1
     size=48
     skewAmount=0
     typefaceFontName=Burbank-Black
     outlineSettings {
      bApplyOutlineToDropShadows=false
      bMiteredCorners=false
      bSeparateFillAlpha=false
      outlineMaterial=null
      outlineSize=0
      outlineColor=<rgba 0 0 0 1>
     }
    }
    shadowOffset=<0 0>
   }
   slot {class=VerticalBoxSlot, properties {horizontalAlignment=HAlign_Center, verticalAlignment=VAlign_Top, padding=<ltrb 0 0 0 32>}}
  }
  Text_Question {
   widgetClassPath=/Script/UMG.RichTextBlock
   bInherited=true
   bIsVariable=true
   parent=VerticalBox_199
   widget=Text_Question
   properties {
    applyLineHeightToBottomLine=false
    justification=Center
    lineHeightPercentage=1.15
    text="[PH] Which weapon <Keyword>would</> you like <Header>unvaulted</>?"
    textStyleSet=&/SurveyRichTextStyles
    wrapTextAt=928
   }
   slot {class=VerticalBoxSlot, properties {horizontalAlignment=HAlign_Center, verticalAlignment=VAlign_Top, padding=<ltrb 0 0 0 32>}}
  }
  Border_ResponseSlot {
   widgetClassPath=&Border
   bIsVariable=true
   parent=VerticalBox_199
   widget=Border_ResponseSlot
   properties {
    clipping=ClipToBounds
    horizontalAlignment=HAlign_Center
    verticalAlignment=VAlign_Center
    background {
     bIsDynamicallyLoaded=false
     drawAs=NoDrawType
     imageType=NoImage
     mirroring=NoMirror
     resourceName=None
     resourceObject=null
     tiling=NoTile
     imageSize=<32 32>
     margin=<ltrb 0 0 0 0>
     outlineSettings {
      bUseBrushTransparency=false
      roundingType=HalfHeightRadius
      width=0
      color {colorUseRule=UseColor_Specified, specifiedColor=<rgba 0 0 0 0>}
      cornerRadii=<0 0 0 0>
     }
     tintColor {colorUseRule=UseColor_Specified, specifiedColor=<rgba 1 1 1 1>}
     uVRegion {bIsValid=false, max=<0 0>, min=<0 0>}
    }
    padding=<ltrb 0 0 0 0>
   }
   slot {class=VerticalBoxSlot, properties {size {sizeRule=Fill, value=1}}}
  }
  HB_Buttons {
   widgetClassPath=&HorizontalBox
   parent=VerticalBox_199
   widget=HB_Buttons
   slot {class=VerticalBoxSlot, properties {horizontalAlignment=HAlign_Center, verticalAlignment=VAlign_Bottom, padding=<ltrb 0 32 0 0>}}
  }
  Button_Skip {
   widgetClassPath=&WBP_UIKit_Button_Regular
   bInherited=true
   parent=HB_Buttons
   widget=Button_Skip
   properties {
    bIsFocusable=false
    minWidthOverride=200
    text=Skip
    triggeringInputAction {dataTable=&InputActionDataTable, rowName=GenericFaceButtonTop}
   }
   slot {class=HorizontalBoxSlot, properties {padding=<ltrb 0 0 24 0>}}
  }
  Switcher_NextButton {widgetClassPath=&WidgetSwitcher, bInherited=true, parent=HB_Buttons, widget=Switcher_NextButton, slot {class=HorizontalBoxSlot}}
  Button_NextQuestion {
   widgetClassPath=/FortUILibrary/UIKit/Buttons/WBP_UIKit_Button_Loud
   bInherited=true
   parent=Switcher_NextButton
   widget=Button_NextQuestion
   properties {
    bIsFocusable=false
    minWidthOverride=200
    text=Next
    triggeringInputAction {dataTable=&InputActionDataTable, rowName=GenericFaceButtonLeft}
   }
   slot {class=WidgetSwitcherSlot}
  }
  Button_Submit {
   widgetClassPath=&WBP_UIKit_Button_Loud
   bInherited=true
   parent=Switcher_NextButton
   widget=Button_Submit
   properties {
    bIsFocusable=false
    minWidthOverride=200
    text=Submit
    triggeringInputAction {dataTable=&InputActionDataTable, rowName=GenericFaceButtonLeft}
   }
   slot {class=WidgetSwitcherSlot}
  }
  Text_Settings {
   widgetClassPath=&CommonTextBlock
   parent=VerticalBox_199
   widget=Text_Settings
   properties {
    style=&TextStyle-Base_C
    text="Providing feedback is optional. You can control these surveys in the settings menu under the \"Account and Privacy\" tab."
    font {
     bForceMonospaced=false
     bMaterialIsStencil=false
     fontMaterial=null
     fontObject=&NotoSans
     letterSpacing=0
     monospacedWidth=1
     size=24
     skewAmount=0
     typefaceFontName=NotoSans-Regular
     outlineSettings {
      bApplyOutlineToDropShadows=false
      bMiteredCorners=false
      bSeparateFillAlpha=false
      outlineMaterial=null
      outlineSize=0
      outlineColor=<rgba 0 0 0 1>
     }
    }
    shadowOffset=<0 0>
   }
   slot {class=VerticalBoxSlot, properties {horizontalAlignment=HAlign_Center, verticalAlignment=VAlign_Bottom, padding=<ltrb 0 32 0 0>}}
  }
  Text_QuestionCount {
   widgetClassPath=&CommonTextBlock
   bInherited=true
   bIsVariable=true
   parent=VerticalBox_199
   widget=Text_QuestionCount
   properties {
    justification=Center
    text="[PH] Question: 1/3"
    font {
     bForceMonospaced=false
     bMaterialIsStencil=false
     fontMaterial=null
     fontObject=&BurbankSmall
     letterSpacing=0
     monospacedWidth=1
     size=15
     skewAmount=0
     typefaceFontName=Black
     outlineSettings {
      bApplyOutlineToDropShadows=false
      bMiteredCorners=false
      bSeparateFillAlpha=false
      outlineMaterial=null
      outlineSize=0
      outlineColor=<rgba 0 0 0 1>
     }
    }
    shadowOffset=<0 0>
   }
   slot {class=VerticalBoxSlot, properties {horizontalAlignment=HAlign_Center, verticalAlignment=VAlign_Bottom, padding=<ltrb 0 32 0 0>}}
  }
 ]
}
